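# First-person launch sound cue for the thrown Super Launch Pad item, in the Fort_Traps sound class.
# A quality-level root picks one of four modulated branches: the first two randomize equally among three launch waves, while the last two play only the first wave.
# A concurrency override allows one instance at a time and prevents new plays while it is active.
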SoundCue /LaunchPadItemGameplay/Items/LaunchPadThrown/Audio/SuperLaunchPad_Launch_1P_Cue
root=QualityLevel_0
nodes [
 {id=Modulator_1, children [Random_4]}
 {id=Modulator_2, children [WavePlayer_17]}
 {id=Modulator_3, children [Random_2]}
 {id=Modulator_4, children [WavePlayer_18]}
 {id=QualityLevel_0, children [Modulator_1,Modulator_3,Modulator_2,Modulator_4]}
 {id=Random_2, children [WavePlayer_7, WavePlayer_8, WavePlayer_16], properties {weights [1,1,1]}}
 {id=Random_4, children [WavePlayer_10, WavePlayer_11, WavePlayer_5], properties {weights [1,1,1]}}
 {id=WavePlayer_10, properties {soundWaveAssetPtr=&/SFX_SuperLaunchPad_Launch_01}}
 {id=WavePlayer_11, properties {soundWaveAssetPtr=&/SFX_SuperLaunchPad_Launch_02}}
 {id=WavePlayer_16, properties {soundWaveAssetPtr=&/SFX_SuperLaunchPad_Launch_03}}
 {id=WavePlayer_17, properties {soundWaveAssetPtr=&SFX_SuperLaunchPad_Launch_01}}
 {id=WavePlayer_18, properties {soundWaveAssetPtr=&SFX_SuperLaunchPad_Launch_01}}
 {id=WavePlayer_5, properties {soundWaveAssetPtr=&SFX_SuperLaunchPad_Launch_03}}
 {id=WavePlayer_7, properties {soundWaveAssetPtr=&SFX_SuperLaunchPad_Launch_01}}
 {id=WavePlayer_8, properties {soundWaveAssetPtr=&SFX_SuperLaunchPad_Launch_02}}
]
properties {
 bOverrideConcurrency=true
 maxDistance=2097152
 soundClassObject=&Fort_Traps
 volumeMultiplier=1
 concurrencyOverrides {maxCount=1, resolutionRule=PreventNew}
}
