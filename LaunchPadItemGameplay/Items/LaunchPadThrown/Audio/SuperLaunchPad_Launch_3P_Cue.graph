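# Third-person sound cue for the Super Launch Pad launch, played with the traps sound class and traps attenuation.
# A Quality Level root selects one of four branches, each passing through a Modulator.
# The first two branches pick at equal weight from three launch wave variations.
# The last two each play only the first variation, so variety depends on the active quality level.
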SoundCue /LaunchPadItemGameplay/Items/LaunchPadThrown/Audio/SuperLaunchPad_Launch_3P_Cue
root=QualityLevel_0
nodes [
 {id=Modulator_4, children [Random_4]}
 {id=Modulator_5, children [WavePlayer_17]}
 {id=Modulator_6, children [WavePlayer_18]}
 {id=Modulator_7, children [Random_3]}
 {id=QualityLevel_0, children [Modulator_4,Modulator_7,Modulator_5,Modulator_6]}
 {id=Random_3, children [WavePlayer_9, WavePlayer_12, WavePlayer_13], properties {weights [1,1,1]}}
 {id=Random_4, children [WavePlayer_10, WavePlayer_11, WavePlayer_8], properties {weights [1,1,1]}}
 {id=WavePlayer_10, properties {soundWaveAssetPtr=&/SFX_SuperLaunchPad_Launch_01}}
 {id=WavePlayer_11, properties {soundWaveAssetPtr=&/SFX_SuperLaunchPad_Launch_02}}
 {id=WavePlayer_12, properties {soundWaveAssetPtr=&SFX_SuperLaunchPad_Launch_02}}
 {id=WavePlayer_13, properties {soundWaveAssetPtr=&/SFX_SuperLaunchPad_Launch_03}}
 {id=WavePlayer_17, properties {soundWaveAssetPtr=&SFX_SuperLaunchPad_Launch_01}}
 {id=WavePlayer_18, properties {soundWaveAssetPtr=&SFX_SuperLaunchPad_Launch_01}}
 {id=WavePlayer_8, properties {soundWaveAssetPtr=&SFX_SuperLaunchPad_Launch_03}}
 {id=WavePlayer_9, properties {soundWaveAssetPtr=&SFX_SuperLaunchPad_Launch_01}}
]
properties {
 attenuationSettings=&Fort_Traps_Att
 maxDistance=6000
 soundClassObject=&Fort_Traps
 volumeMultiplier=1
}
