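; Thrown launch pad trap Blueprint that extends a parent trap with placement, launch and death behavior.
; On placement it plays audio, shows a hostile sound indicator the parent skips for thrown items, and calls Mark on remotes to auto-ping the team.
; It dies when hit by actors tagged as physics objects.
; Launches and vehicle overlaps spawn Niagara effects; death hides the pad and destroys its collision, leaving debug prints.
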(event EventOnPlaced
  (bind _trap_placed_sound (Variables|Audio|GetTrap_Placed_Sound))
  (Collision|BindEventtoOnActorHit (delegate Custom|When%20Hit))
  (Trap|AddTriggerComponent (Variables|Default|GetS_Athena_Launchpad_Collision))
  (Audio|PlaySoundatLocation _trap_placed_sound (Transformation|GetWorldLocation (Variables|Default|GetPlacementSoundLocation)))
  ;[1 1 1] There is a Sound Indicator call in the parent class that is never called, because the Thrown item isn't being constructed by a player. We call another one here instead.
  (Fortnite|ShowSoundIndicatorLocalforTeam :Location (Transformation|GetActorLocation) :MaxAudibleDistance (Variables|/Script/Engine.SoundBase|GetMaxDistance _trap_placed_sound) :IndicatorType "Generic" :Team (Team|GetActorTeam self) :Affiliations (Utilities|Array|MakeArray "Hostile") :Tint "(R=0.000000,G=0.000000,B=0.000000,A=0.000000)" :OverrideIcon "/Game/UI/Foundation/Textures/Icons/Navigation/t_icon_nav_mobile_audioviz_building_32.t_icon_nav_mobile_audioviz_building_32" :AttenuationMaxRangeMultiplier 1.0 :AttenuationIntensityMultiplier 1.0)
  ;[1 1 1] Auto ping the launch pad to team
  (Networking|SwitchHasAuthority
    (:Remote
      (CallFunction|Mark))))

(event Custom|When%20Hit
  (params (SelfActor :type Actor%20Object%20Reference) (OtherActor :type Actor%20Object%20Reference) (NormalImpulse :type Vector) (Hit :type Hit%20Result%20Structure%20%28by%20ref%29))
  (bind _returnvalue (GameplayTags|HasAnyMatchingGameplayTags(Message) OtherActor (Variables|Default|GetPhysics%20Object%20Tags)))
  (if _returnvalue
    (Gameplay|Die :DamageCauser OtherActor)))

(event EventTriggerLaunchEffects
  (params (Pawn :type Fort%20Pawn%20Object%20Reference))
  ;[1 1 1] LAUNCH FX
  (|CallParentFunction Pawn)
  (Niagara|SpawnSystematLocation (Variables|FX|GetPS_LaunchFX) (Transformation|GetWorldLocation (Variables|Default|GetPlacementSoundLocation))))

(event Custom|Event%20Vehicle%20Trigger%20FX
  ;[1 1 1] LAUNCH FX
  (Niagara|SpawnSystematLocation (Variables|FX|GetPS_LaunchFX) (Transformation|GetWorldLocation (Variables|Default|GetPlacementSoundLocation))))

(event OnComponentBeginOverlap(S_Athena_Launchpad_Collision)
  (params (OverlappedComponent :type Primitive%20Component%20Object%20Reference) (OtherActor :type Actor%20Object%20Reference) (OtherComp :type Primitive%20Component%20Object%20Reference) (OtherBodyIndex :type Integer) (bFromSweep :type Boolean) (SweepResult :type Hit%20Result%20Structure%20%28by%20ref%29))
  ;[1 1 1] If a Vehicle hits the pad, spawn Launch FX for that too
  (Utilities|Casting|CastToFortAthenaVehicle OtherActor)
  (CallFunction|EventVehicleTriggerFX))

(event EventOnDeathServer
  (params (Damage :type Float%20%28single-precision%29) (DamageTags :type Gameplay%20Tag%20Container%20Structure%20%28by%20ref%29) (Momentum :type Vector :default "0, 0, 0") (HitInfo :type Hit%20Result%20Structure%20%28by%20ref%29) (InstigatedBy :type Controller%20Object%20Reference) (DamageCauser :type Actor%20Object%20Reference) (EffectContext :type Gameplay%20Effect%20Context%20Handle%20Structure))
  (Development|PrintString "Died" true true "(R=1.000000,G=0.000000,B=0.156583,A=1.000000)" 5.0)
  (Rendering|SetActorHiddenInGame true)
  (Components|DestroyComponent (Variables|Default|GetS_Athena_Launchpad_Collision))
  (|CallParentFunction Damage DamageTags Momentum HitInfo InstigatedBy DamageCauser EffectContext))

(event EventOnDeathPlayEffects
  (params (Damage :type Float%20%28single-precision%29) (DamageTags :type Gameplay%20Tag%20Container%20Structure%20%28by%20ref%29) (Momentum :type Vector :default "0, 0, 0") (HitInfo :type Hit%20Result%20Structure%20%28by%20ref%29) (InstigatedBy :type Fort%20Pawn%20Object%20Reference) (DamageCauser :type Actor%20Object%20Reference) (EffectContext :type Gameplay%20Effect%20Context%20Handle%20Structure))
  (Development|PrintString "Died" true true "(R=1.000000,G=0.000000,B=0.156583,A=1.000000)" 5.0)
  (Rendering|SetActorHiddenInGame true)
  (Components|DestroyComponent (Variables|Default|GetS_Athena_Launchpad_Collision))
  (|CallParentFunction Damage DamageTags Momentum HitInfo InstigatedBy DamageCauser EffectContext))

(event EventDestroyed
  (Development|PrintString "Destroyed"))

(fn Mark
  (bind _asfort_player_controller_athena (Utilities|Casting|CastToFortPlayerControllerAthena (Actor|GetOwner (Building|GetOwningPlayerState))))
  (Fortnite|RegisterSoundIndicatorPlacementContext _asfort_player_controller_athena self))

(fn ConstructionScript
  (|CallParentFunction))
